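# Niagara system for the Caller emote: a single pale-blue glow sprite offset slightly from its origin.
# One burst-spawned particle pulses its brightness with a sine on emitter age, with alpha scaled by time of day and front-end versus gameplay.
# Camera-distance remaps drive sprite size curves, presumably keeping the glow readable from afar.
# It also writes flare color, size and owner-axis orientation attributes, though only one sprite renderer is listed.
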NiagaraSystem /BRCosmetics/Animation/Game/MainPlayer/Emotes/Caller/Effects/NS_CallerEmote
dependencies {usedDataInterfaces [&NiagaraDataInterfaceCurve], usedDynamicInputs [], usedModules [&SystemState]}
emitters {CallerEmote {
  emitterUpdateScript [
   &EmitterState(
    Inactive Response=::NewEnumerator1
    Life Cycle Mode=::NewEnumerator1
    Loop Behavior=::NewEnumerator1
    Loop Duration Mode=::NewEnumerator1
    UseLoopDelay=false
   )
   &SpawnBurst_Instantaneous(Spawn Count=1, Use Spawn Probability=false)
  ]
  particleSpawnScript [
   &InitializeParticle(
    Color=<rgba 0.4555 0.643 0.75 40>
    Position Offset=<-1.5 1 0>
    Sprite Rotation Mode=::NewEnumerator0
    Sprite Size Mode=::NewEnumerator3
    Uniform Sprite Size=70
    UsePositionOffset=true
    Write Material Random=true
   )
   /SetVariables(Particles.FlareColor=<rgba 10 10 10 0.5>) {isSetParameters=true}
   /SetVariables(Particles.FlareSize:Vector2f=<40 10>) {isSetParameters=true}
   /SetVariables(Particles.FlareAlignement:Vector3f=$Engine.Owner.SystemZAxis) {isSetParameters=true}
   /SetVariables(Particles.FlareFacing:Vector3f=$Engine.Owner.SystemYAxis) {isSetParameters=true}
   &CameraOffset(Camera Offset Amount=2)
  ]
  particleUpdateScript [
   &ParticleState(Kill Particles When Lifetime Has Elapsed=false, Let Infinitely Lived Particles Die When Emitter Deactivates=true)
   &ScaleColor(
    Scale Alpha=&Multiply_Float(
     A=&FNTimeOfDayScalar(Night Scale=0.1, Morning Scale=0.5, Evening Scale=0.1)
     B=&FN_FrontEndFloatModifier(Gameplay=0.5, Front End=1)
    )
    ScaleA=true
    ScaleRGB=true
   )
   /SetVariables(Particles.Color=&MultiplyLinearColor(
     Linear Color=$Particles.Color
     Scale Factor=&MakeFloat4FromVectorAndFloat(
      "Vector (RGB)"=&VectorFromFloat(Value=&Sine(Normalized Angle=$Emitter.Age, Period=0.4, Scale=0.2, Bias=0.7))
      "Float (Alpha)"=&Multiply_Float(
       A=&FNTimeOfDayScalar(Night Scale=0.4, Morning Scale=0.1, Day Scale=0.1, Evening Scale=0.4)
       B=&FN_FrontEndFloatModifier(Gameplay=0.5, Front End=1)
      )
     )
    )) {isSetParameters=true}
   &DynamicMaterialParameters(
    Index 0 Param 1=1
    Module.Param0WriteEnabled=true
    Module.Param0WriteEnabled1=true
    Module.Param1WriteEnabled=true
    Module.Param1WriteEnabled1=true
    Module.Param2WriteEnabled=true
    Module.Param2WriteEnabled1=true
    Module.Param3WriteEnabled=true
    Module.Param3WriteEnabled1=true
    Write Parameter Index 1=true
   )
   &ScaleSpriteSize(
    Uniform Curve Scale=&FN_LOD_Distance_Remap(Start Distance=600, End Distance=20000, Minimum Output=0.04, Maximum Output=0.8)
    Uniform Curve Sprite Scale=DataInterface() {propertyValues {Curve {keys [<at 0 0>, {arriveTangent=193.1823, leaveTangent=193.1823, time=1, value=50}]}}}
   ) {module=ScaleSpriteSize002}
   /SetVariables(Particles.SpriteSize001:Vector2f=&Multiply_Vector2D_ByFloat(
     Float=&FloatFromCurve(
      FloatCurve=DataInterface() {propertyValues {Curve {keys [<at 0 1>, {arriveTangent=193.1823, leaveTangent=193.1823, time=1, value=2}]}}}
      Scale Curve=&FN_LOD_Distance_Remap(Start Distance=600, End Distance=10000, Minimum Output=0.6, Maximum Output=8)
     )
     Vector2D=<70 5>
    )) {isSetParameters=true}
  ]
  rendererClasses [&NiagaraSpriteRendererProperties]
  renderers [{rendererClass=&NiagaraSpriteRendererProperties}]
 }}
summary {
 systemName=NS_CallerEmote
 emitters [{emitterName=CallerEmote, rendererClasses [&NiagaraSpriteRendererProperties]}]
 userVariables [
  User.bIsFrontEnd {type=NiagaraBool}
  User.bIsFrontEndPreview {type=NiagaraBool}
  User.PlayerSpeed {type=NiagaraFloat, defaultValue=0, description=""}
 ]
}
